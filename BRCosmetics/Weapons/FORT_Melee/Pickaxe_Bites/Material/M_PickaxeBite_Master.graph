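# Master material for the Pickaxe Bites melee cosmetic, built between the shared character PreFX and PostFX functions.
# MF_FlowingNormals adds liquid normals, masked by offset AO times a sphere mask on pre-skinned local position.
# A Fresnel of the transformed liquid normal samples a thin-film spectrum, which is softened, desaturated and multiplied into a brightened base color.
# Original roughness remaps between two parameters before PostFX.
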Material /BRCosmetics/Weapons/FORT_Melee/Pickaxe_Bites/Material/M_PickaxeBite_Master
graph {
 properties {
  bUseMaterialAttributes=true
  bUsedWithMeshDeformer=true
  bUsedWithSkeletalMesh=true
  pixelDepthOffsetMode=PDOM_Legacy
  usedShadingModels=Default Lit
 }
 nodes [
  Constant() {r=1}
  n1 = MaterialFunctionCall() {materialFunction=&BaseFN_CharacterFunction_PreFX}
  n2 = Add(A=@n1.AO, B=ScalarParameter() {group=PickaxeBites, parameterName=AO_indesity}) {constB=0.3}
  n3 = Constant3Vector()
  n4 = ScalarParameter() {defaultValue=20, group=PickaxeBites, parameterName=liquidNormalRadius}
  n5 = ScalarParameter() {defaultValue=1, group=PickaxeBites, parameterName=liquidNormalHardness}
  n6 = GetMaterialAttributes(None=@n1.Result) {attributeGetTypes [#1, #2]}
  n7 = MaterialFunctionCall!(
   Alpha=Multiply(
    A=Saturate(None=@n2)
    B=SphereMask(A=@n1.Pre skinned Local Position, B=@n3, Radius=@n4, Hardness=@n5) {attenuationRadius=20, hardnessPercent=1}
   )
   LiquidNormalMask=Constant4Vector() {constant=<rgba 1 1 1 1>}.RGBA
   In=@n6.MaterialAttributes
  ) {materialFunction=/Game/Global/MaterialFunctions/MF_FlowingNormals}
  n8 = ScalarParameter() {defaultValue=0.2, group=PickaxeBites, parameterName=RoughnessMin_a}
  n9 = ScalarParameter() {defaultValue=0.3, group=PickaxeBites, parameterName=RoughnessMax_B}
  MaterialFunctionCall(
   AO=@n1.AO
   Distance Blend=@n1.AO
   Fresnel=@n1.AO
   Gmap=@n1.AO
   Material=SetMaterialAttributes!(MaterialAttributes=@n7.Result, Roughness=LinearInterpolate(A=@n8, B=@n9, Alpha=@n6.Roughness), Base Color=NamedRerouteUsage()) {attributeSetTypes [#2, #1]}
  ) {materialFunction=&BaseFN_CharacterFunction_PostFX}
  NamedRerouteDeclaration(None=@n7.LiquidNormal) {name=LiquidNormal, nodeColor=<rgba 1 0 0.6797 1>}
  NamedRerouteDeclaration(None=@n6.Base Color) {name=baseColor, nodeColor=<rgba 0 1 0.2031 1>}
  n10 = NamedRerouteUsage()
  n11 = Constant3Vector() {constant=<rgba 0.9323 0.9323 0.9323 0>}
  n12 = ScalarParameter() {defaultValue=0.75, group=PickaxeBites, parameterName=basecolor_brightness}
  n13 = ScalarParameter() {defaultValue=3, parameterName=Param_2}
  n14 = Transform(None=NamedRerouteUsage())
  n15 = LinearInterpolate(
   A=TextureSampleParameter2D(UVs=Fresnel(ExponentIn=@n13, Normal=@n14)) {parameterName=Param, texture=/Game/Global/Textures/Default/Thinfilm/T_ThinFilm_Spectrum_COLOR}.RGB
   Alpha=ScalarParameter() {defaultValue=0.85}
  )
  n16 = ScalarParameter() {defaultValue=0.75, parameterName=Param_1}
  NamedRerouteDeclaration(None=Multiply(A=LinearInterpolate(A=@n10, B=@n11, Alpha=@n12), B=Desaturation(None=@n15, Fraction=@n16))) {name=out base color, nodeColor=<rgba 1 0.7266 0 1>}
 ]
}
