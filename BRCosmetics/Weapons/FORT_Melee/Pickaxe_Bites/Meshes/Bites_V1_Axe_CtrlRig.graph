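# Control Rig for the Bites V1 pickaxe that procedurally animates the seven dyn_tendrill bone chains attached to its two-handed handle.
# The forward solve runs one ChainHarmonicsPerItem unit per chain in sequence, each starting from that chain's root bone.
# Every unit layers reach, wave and zero-gravity pendulum motion.
# Pendulum settings are shared, while speed, reach and wave tuning vary by chain, with no variables driving them.
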ControlRigBlueprint /BRCosmetics/Weapons/FORT_Melee/Pickaxe_Bites/Meshes/Bites_V1_Axe_CtrlRig
graphs [
 RigVMFunctionLibrary {links [], nodes []}
 RigVMModel {nodes [
   RigUnit_BeginExecution = &RigUnit_BeginExecution()
   /Script/ControlRig.RigUnit_ChainHarmonicsPerItem(
    ChainRoot=(Type=Bone,Name="dyn_tendrill_bottom_0")
    DrawWorldOffset=(Rotation=(X=0,Y=0,Z=0,W=1),Translation=(X=0,Y=0,Z=0),Scale3D=(X=1,Y=1,Z=1))
    Pendulum=(bEnabled=True,PendulumStiffness=2,PendulumGravity=(X=0,Y=0,Z=0),PendulumBlend=0.75,PendulumDrag=0.98,PendulumMaximum=0.1,UnwindAxis=(X=0,Y=1,Z=0),UnwindMinimum=0.2,UnwindMaximum=0.05)
    Reach=(bEnabled=True,ReachTarget=(X=0,Y=0,Z=0),ReachAxis=(X=1.5,Y=0.25,Z=0),ReachMinimum=-1,ReachMaximum=0.2)
    Speed=(X=2,Y=0.5,Z=3)
    Wave=(bEnabled=True,WaveFrequency=(X=1,Y=1,Z=0.8),WaveAmplitude=(X=0,Y=2.25,Z=0),WaveOffset=(X=0,Y=1,Z=2),WaveNoise=(X=0,Y=0,Z=0),WaveMinimum=0,WaveMaximum=0.1)
    WaveCurve=(EditorCurveData=(Keys=((),(Time=1,Value=1)),DefaultValue=340282346638528859811704183484516925440,PreInfinityExtrap=RCCE_Constant,PostInfinityExtrap=RCCE_Constant))
    bDrawDebug=false
    ExecutePin=@RigUnit_BeginExecution.ExecutePin
   ) {is=struct RigVMUnitNode}
   > &RigUnit_ChainHarmonicsPerItem(
    ChainRoot=(Type=Bone,Name="dyn_tendrill_frontLeft_0")
    DrawWorldOffset=(Rotation=(X=0,Y=0,Z=0,W=1),Translation=(X=0,Y=0,Z=0),Scale3D=(X=1,Y=1,Z=1))
    Pendulum=(bEnabled=True,PendulumStiffness=2,PendulumGravity=(X=0,Y=0,Z=0),PendulumBlend=0.75,PendulumDrag=0.98,PendulumMaximum=0.1,UnwindAxis=(X=0,Y=1,Z=0),UnwindMinimum=0.2,UnwindMaximum=0.05)
    Reach=(bEnabled=True,ReachTarget=(X=0,Y=0,Z=0),ReachAxis=(X=1,Y=0,Z=0))
    Speed=(X=2,Y=1.5,Z=3)
    Wave=(bEnabled=True,WaveFrequency=(X=1,Y=0.6,Z=0.8),WaveAmplitude=(X=0,Y=1,Z=0),WaveOffset=(X=0,Y=1,Z=2),WaveNoise=(X=0,Y=0,Z=0),WaveMaximum=1)
    WaveCurve=(EditorCurveData=(Keys=((),(Time=1,Value=1)),DefaultValue=340282346638528859811704183484516925440,PreInfinityExtrap=RCCE_Constant,PostInfinityExtrap=RCCE_Constant))
    bDrawDebug=false
   )
   > &RigUnit_ChainHarmonicsPerItem(
    ChainRoot=(Type=Bone,Name="dyn_tendrill_frontMid_right_0")
    DrawWorldOffset=(Rotation=(X=0,Y=0,Z=0,W=1),Translation=(X=0,Y=0,Z=0),Scale3D=(X=1,Y=1,Z=1))
    Pendulum=(bEnabled=True,PendulumStiffness=2,PendulumGravity=(X=0,Y=0,Z=0),PendulumBlend=0.75,PendulumDrag=0.98,PendulumMaximum=0.1,UnwindAxis=(X=0,Y=1,Z=0),UnwindMinimum=0.2,UnwindMaximum=0.05)
    Reach=(bEnabled=True,ReachTarget=(X=0,Y=0,Z=0),ReachAxis=(X=1,Y=0,Z=0))
    Speed=(X=2,Y=0.5,Z=7)
    Wave=(bEnabled=True,WaveFrequency=(X=1,Y=0.6,Z=0.8),WaveAmplitude=(X=0,Y=1,Z=0),WaveOffset=(X=0,Y=1,Z=2),WaveNoise=(X=0,Y=0,Z=0),WaveMaximum=1)
    WaveCurve=(EditorCurveData=(Keys=((),(Time=1,Value=1)),DefaultValue=340282346638528859811704183484516925440,PreInfinityExtrap=RCCE_Constant,PostInfinityExtrap=RCCE_Constant))
    bDrawDebug=false
   )
   > &RigUnit_ChainHarmonicsPerItem(
    ChainRoot=(Type=Bone,Name="dyn_tendrill_frontRight_0")
    DrawWorldOffset=(Rotation=(X=0,Y=0,Z=0,W=1),Translation=(X=0,Y=0,Z=0),Scale3D=(X=1,Y=1,Z=1))
    Pendulum=(bEnabled=True,PendulumStiffness=2,PendulumGravity=(X=0,Y=0,Z=0),PendulumBlend=0.75,PendulumDrag=0.98,PendulumMaximum=0.1,UnwindAxis=(X=0,Y=1,Z=0),UnwindMinimum=0.2,UnwindMaximum=0.05)
    Reach=(bEnabled=True,ReachTarget=(X=0,Y=0,Z=0),ReachAxis=(X=1,Y=0,Z=0))
    Speed=(X=7,Y=2,Z=7)
    Wave=(bEnabled=True,WaveFrequency=(X=1,Y=0.6,Z=0.8),WaveAmplitude=(X=0,Y=1,Z=0),WaveOffset=(X=0,Y=1,Z=2),WaveNoise=(X=0,Y=0,Z=0),WaveMaximum=1)
    WaveCurve=(EditorCurveData=(Keys=((),(Time=1,Value=1)),DefaultValue=340282346638528859811704183484516925440,PreInfinityExtrap=RCCE_Constant,PostInfinityExtrap=RCCE_Constant))
    bDrawDebug=false
   )
   > &RigUnit_ChainHarmonicsPerItem(
    ChainRoot=(Type=Bone,Name="dyn_tendrill_left_0")
    DrawWorldOffset=(Rotation=(X=0,Y=0,Z=0,W=1),Translation=(X=0,Y=0,Z=0),Scale3D=(X=1,Y=1,Z=1))
    Pendulum=(bEnabled=True,PendulumStiffness=2,PendulumGravity=(X=0,Y=0,Z=0),PendulumBlend=0.75,PendulumDrag=0.98,PendulumMaximum=0.1,UnwindAxis=(X=0,Y=1,Z=0),UnwindMinimum=0.2,UnwindMaximum=0.05)
    Reach=(bEnabled=True,ReachTarget=(X=0,Y=0,Z=0),ReachAxis=(X=0.5,Y=0.5,Z=0.2),ReachMinimum=0.5)
    Speed=(X=2,Y=0.5,Z=3)
    Wave=(bEnabled=True,WaveFrequency=(X=1,Y=0.6,Z=0.8),WaveAmplitude=(X=0,Y=1,Z=0),WaveOffset=(X=0,Y=1,Z=2),WaveNoise=(X=0,Y=0,Z=0),WaveMaximum=1)
    WaveCurve=(EditorCurveData=(Keys=((),(Time=1,Value=1)),DefaultValue=340282346638528859811704183484516925440,PreInfinityExtrap=RCCE_Constant,PostInfinityExtrap=RCCE_Constant))
    bDrawDebug=false
   )
   > &RigUnit_ChainHarmonicsPerItem(
    ChainRoot=(Type=Bone,Name="dyn_tendrill_rightBack_0")
    DrawWorldOffset=(Rotation=(X=0,Y=0,Z=0,W=1),Translation=(X=0,Y=0,Z=0),Scale3D=(X=1,Y=1,Z=1))
    Pendulum=(bEnabled=True,PendulumStiffness=2,PendulumGravity=(X=0,Y=0,Z=0),PendulumBlend=0.75,PendulumDrag=0.98,PendulumMaximum=0.1,UnwindAxis=(X=0,Y=1,Z=0),UnwindMinimum=0.2,UnwindMaximum=0.05)
    Reach=(bEnabled=True,ReachTarget=(X=0,Y=0,Z=0),ReachAxis=(X=1,Y=0,Z=0),ReachMinimum=-0.5)
    Speed=(X=3,Y=1,Z=1)
    Wave=(bEnabled=True,WaveFrequency=(X=1,Y=0.6,Z=0.8),WaveAmplitude=(X=0,Y=1,Z=1),WaveOffset=(X=0,Y=1,Z=2),WaveNoise=(X=0,Y=0,Z=0),WaveMaximum=1)
    WaveCurve=(EditorCurveData=(Keys=((),(Time=1,Value=1)),DefaultValue=340282346638528859811704183484516925440,PreInfinityExtrap=RCCE_Constant,PostInfinityExtrap=RCCE_Constant))
    bDrawDebug=false
   )
   > &RigUnit_ChainHarmonicsPerItem(
    ChainRoot=(Type=Bone,Name="dyn_tendrill_top_0")
    DrawWorldOffset=(Rotation=(X=0,Y=0,Z=0,W=1),Translation=(X=0,Y=0,Z=0),Scale3D=(X=1,Y=1,Z=1))
    Pendulum=(bEnabled=True,PendulumStiffness=2,PendulumGravity=(X=0,Y=0,Z=0),PendulumBlend=0.75,PendulumDrag=0.98,PendulumMaximum=0.1,UnwindAxis=(X=0,Y=1,Z=0),UnwindMinimum=0.2,UnwindMaximum=0.05)
    Reach=(bEnabled=True,ReachTarget=(X=0,Y=0,Z=0),ReachAxis=(X=0.5,Y=0.5,Z=0.5),ReachMinimum=-1,ReachMaximum=0)
    Speed=(X=0.5,Y=1,Z=2)
    Wave=(bEnabled=True,WaveFrequency=(X=1,Y=0.6,Z=0.8),WaveAmplitude=(X=1,Y=1,Z=0),WaveOffset=(X=0,Y=1,Z=2),WaveNoise=(X=0,Y=0,Z=0),WaveMaximum=1)
    WaveCurve=(EditorCurveData=(Keys=((),(Time=1,Value=1)),DefaultValue=340282346638528859811704183484516925440,PreInfinityExtrap=RCCE_Constant,PostInfinityExtrap=RCCE_Constant))
    bDrawDebug=false
   )
  ]}
]
hierarchy [
 b_AR_Root {loc=<0 0 0>, rot=<0 0 0 1>}
 Scaler_Joint {parent=b_AR_Root, loc=<0 0 0>, rot=<0 0 0.7071 -0.7071>}
 Melee_TwoHanded_Handle {parent=Scaler_Joint, loc=<0 0 0>, rot=<0 0 0 1>}
 dyn_tendrill_rightBack_0 {parent=Melee_TwoHanded_Handle, loc=<4.8458 -15.5353 -0.7112>, rot=<-0.3821 -0.5902 -0.3907 -0.5942>}
 dyn_tendrill_rightBack_1 {parent=dyn_tendrill_rightBack_0, loc=<4.0735 0 0.1206>, rot=<-0.1121 0.0141 -0.0669 -0.9913>}
 dyn_tendrill_top_0 {parent=Melee_TwoHanded_Handle, loc=<0.7797 -14.6105 2.5919>, rot=<-0.1475 -0.0818 -0.6833 -0.7103>}
 dyn_tendrill_top_1 {parent=dyn_tendrill_top_0, loc=<4.1347 0.2721 -0.1726>, rot=<0.0003 -0.0026 -0.0394 -0.9992>}
 dyn_tendrill_top_2 {parent=dyn_tendrill_top_1, loc=<2.9037 -0.0989 -0.3489>, rot=<-0.036 0.153 -0.047 -0.9864>}
 dyn_tendrill_left_0 {parent=Melee_TwoHanded_Handle, loc=<-4.4149 -19.3186 -0.0213>, rot=<0.6567 0.5995 -0.3167 -0.3302>}
 dyn_tendrill_left_1 {parent=dyn_tendrill_left_0, loc=<2.6728 0.042 -0.0732>, rot=<0.0167 0.0537 -0.0122 -0.9983>}
 dyn_tendrill_left_2 {parent=dyn_tendrill_left_1, loc=<3.738 -0.0079 0.0128>, rot=<-0.256 0.0475 0.1476 -0.9542>}
 dyn_tendrill_frontLeft_0 {parent=Melee_TwoHanded_Handle, loc=<-4.2658 -33.5726 0.8621>, rot=<0.3931 -0.6361 -0.5933 0.2979>}
 dyn_tendrill_frontLeft_1 {parent=dyn_tendrill_frontLeft_0, loc=<1.4138 0 0>, rot=<0 0 0 1>}
 dyn_tendrill_frontRight_0 {parent=Melee_TwoHanded_Handle, loc=<4.0739 -35.7511 0.015>, rot=<-0.6838 -0.1343 -0.1329 -0.7048>}
 dyn_tendrill_frontRight_1 {parent=dyn_tendrill_frontRight_0, loc=<2.4737 0 0>, rot=<0.039 -0.4425 0.0787 -0.8924>}
 dyn_tendrill_frontMid_right_0 {parent=Melee_TwoHanded_Handle, loc=<5.2156 -28.5405 -0.5275>, rot=<-0.7058 0.0428 0.0428 -0.7058>}
 dyn_tendrill_frontMid_right_1 {parent=dyn_tendrill_frontMid_right_0, loc=<0.9699 0 0>, rot=<0 -0.4063 0 -0.9137>}
 dyn_tendrill_bottom_0 {parent=Melee_TwoHanded_Handle, loc=<0.0321 -14.7388 -4.0918>, rot=<0.0294 -0.1837 -0.7331 -0.6541>}
 dyn_tendrill_bottom_1 {parent=dyn_tendrill_bottom_0, loc=<5.6229 0 0>, rot=<0.0566 0.2683 -0.1612 -0.9481>}
 dyn_tendrill_bottom_2 {parent=dyn_tendrill_bottom_1, loc=<5.8139 0 0>, rot=<0.009 0.0003 0.1178 -0.993>}
 dyn_tendrill_bottom_3 {parent=dyn_tendrill_bottom_2, loc=<3.0727 0 0>, rot=<0.0133 0.0175 0.1294 -0.9914>}
]
