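# Material function for the Bites pickaxe that builds animated flowing-liquid normals as four separate outputs.
# Two tiled, rotated UV samples of a flow-warp normal map pan at slow and fast speeds, with FXHandMask blending between opposite default flow directions.
# Two world-aligned projections of warp normals scroll over time at different sizes and speeds, with a position input (ps_postion) declared for a named reroute.
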MaterialFunction /BRCosmetics/Weapons/FORT_Melee/Pickaxe_Bites/Material/MF_FlowingNormals_Bites
graph {nodes [
  FunctionOutput(None=NamedRerouteUsage()) {outputName=LiquidNormal}
  n1 = LinearInterpolate(
   A=VectorParameter() {group=LiquidNormal, parameterName=LQN_fast_normal2, defaultValue=<rgba 0 0.05 0 1>}.RGB
   B=VectorParameter() {group=LiquidNormal, parameterName=LQN_Fast_normal2_hand, defaultValue=<rgba 0 -0.05 0 1>}.RGB
   Alpha=NamedRerouteUsage()
  )
  n2 = Add(
   A=Multiply(A=TextureCoordinate(), B=ScalarParameter() {defaultValue=3.5, group=LiquidNormal, parameterName=LQN_UVTile2})
   B=Multiply(A=ComponentMask(None=@n1) {g=true, r=true}, B=Time())
  )
  n3 = NamedRerouteUsage()
  n4 = TextureObjectParameter() {
   group=LiquidNormal
   parameterName=LQN_NormalTexture
   samplerType=SAMPLERTYPE_Normal
   texture=/Game/Global/Textures/Default/Noise/T_FlowWarp_N
  }
  FunctionOutput(None=TextureSample(
    UVs=MaterialFunctionCall(UVs=@n2, "Rotation Angle (0-1)"=@n3) {materialFunction=&CustomRotator}.Rotated Values
    Tex=@n4
   ).RGB) {outputName=LiquidNormal_2}
  n5 = TextureObjectParameter() {group=LiquidNormal, parameterName=LQN_WarpTexture, samplerType=SAMPLERTYPE_Normal, texture=&T_FlowWarp_N}
  n6 = NamedRerouteUsage()
  n7 = Multiply(A=Time(), B=VectorParameter() {group=LiquidNormal, parameterName=LQN_WarpUVSpeed, defaultValue=<rgba 0 0 1.5 1>}.RGB)
  FunctionOutput(None=MaterialFunctionCall(
    TextureObject=@n5
    TextureSize=ScalarParameter() {defaultValue=90, group=LiquidNormal, parameterName=LQN_WarpTextureSize}
    WorldPosition=Add(A=@n6, B=@n7)
    World Space Normal=NamedRerouteUsage()
    ProjectionTransitionContrast=ScalarParameter() {group=LiquidNormal, parameterName=LQN_WarpProjectionContrast, sliderMax=1}
   ) {materialFunction=&WorldAlignedTexture}.XYZ Texture) {outputName=WP_normal}
  n8 = TextureObjectParameter() {group=LiquidNormal, parameterName=LQN_WarpTexture, samplerType=SAMPLERTYPE_Normal, texture=&T_FlowWarp_N}
  n9 = NamedRerouteUsage()
  n10 = Multiply(A=Time(), B=VectorParameter() {parameterName=LQN_WarpUVSpeed2, defaultValue=<rgba 0 0 3 1>}.RGB)
  FunctionOutput(None=MaterialFunctionCall(
    TextureObject=@n8
    TextureSize=ScalarParameter() {defaultValue=40, group=LiquidNormal, parameterName=LQN_WarpTextureSize2}
    WorldPosition=Add(A=@n9, B=@n10)
    World Space Normal=NamedRerouteUsage()
   ) {materialFunction=&WorldAlignedTexture}.XYZ Texture) {outputName=WP_normal_2}
  NamedRerouteDeclaration(None=ScalarParameter() {defaultValue=0.5, group=LiquidNormal, parameterName=UVRotation}) {name=UV Rotation, nodeColor=<rgba 0.9922 0 1 1>}
  n11 = FunctionInput() {bUsePreviewValueAsDefault=true, inputName=FXHandMask, inputType=FunctionInput_Scalar}
  NamedRerouteDeclaration(None=@n11) {name=UV_SpeedMask, nodeColor=<rgba 0 1 0.3203 1>}
  NamedRerouteDeclaration() {name=PS_Normal, nodeColor=<rgba 1 0 0.8437 1>}
  NamedRerouteDeclaration(None=FunctionInput() {bUsePreviewValueAsDefault=true, inputName=ps_postion}) {name=PS_Position, nodeColor=<rgba 1 0 0.1641 1>}
  n12 = LinearInterpolate(
   A=VectorParameter() {group=LiquidNormal, parameterName=LQN_slow_normal1, defaultValue=<rgba 0 0.03 0 1>}.RGB
   B=VectorParameter() {group=LiquidNormal, parameterName=LQN_slow_hand, defaultValue=<rgba 0 -0.03 0 1>}.RGB
   Alpha=@n11
  )
  n13 = Add(
   A=Multiply(A=TextureCoordinate(), B=ScalarParameter() {defaultValue=2, group=LiquidNormal, parameterName=LQN_UVTile})
   B=Multiply(A=ComponentMask(None=@n12) {g=true, r=true}, B=Time())
  )
  n14 = NamedRerouteUsage()
  n15 = TextureObjectParameter() {
   group=LiquidNormal
   parameterName=LQN_NormalTexture
   samplerType=SAMPLERTYPE_Normal
   texture=&T_FlowWarp_N
  }
  NamedRerouteDeclaration(None=TextureSample(
    UVs=MaterialFunctionCall(UVs=@n13, "Rotation Angle (0-1)"=@n14) {materialFunction=&CustomRotator}.Rotated Values
    Tex=@n15
   ).RGB) {name=LiquidNormal, nodeColor=<rgba 1 0.7266 0 1>}
 ]}
