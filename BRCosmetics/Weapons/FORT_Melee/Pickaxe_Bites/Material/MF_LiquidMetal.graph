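# Material function that adds a flowing liquid-metal look to incoming material attributes for the Bites pickaxe cosmetic.
# Flowing normals from MF_FlowingNormals_Bites are flattened by a hand-mask texture and blended over the normal; Low, Medium and High quality skip the second normal blend.
# Base color gets a desaturated Fresnel thin-film tint, brightness and AO.
# Masks blend in these results, full metallic and a roughness parameter.
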MaterialFunction /BRCosmetics/Weapons/FORT_Melee/Pickaxe_Bites/Material/MF_LiquidMetal
graph {nodes [
  n1 = GetMaterialAttributes(None=FunctionInput() {inputName=material, inputType=FunctionInput_MaterialAttributes}) {attributeGetTypes [
    #1
    #2
    #3
    #4
   ]}
  n2 = SetMaterialAttributes!(
   MaterialAttributes=@n1.MaterialAttributes
   BaseColor=NamedRerouteUsage()
   Metallic=LinearInterpolate(A=@n1.Metallic, B=Constant() {r=1}, Alpha=NamedRerouteUsage())
   Roughness=LinearInterpolate(A=@n1.Roughness, B=ScalarParameter() {defaultValue=0.23, group=LiquidNormal, parameterName=Roughness}, Alpha=NamedRerouteUsage())
   Normal=NamedRerouteUsage()
  ) {attributeSetTypes [
    #1
    #2
    #3
    #4
   ]}
  FunctionOutput(None=@n2) {outputName=material}
  NamedRerouteDeclaration(None=@n1.BaseColor) {name=BaseColor, nodeColor=<rgba 0 1 0.7422 1>}
  NamedRerouteDeclaration(None=FunctionInput() {bUsePreviewValueAsDefault=true, inputName=FxMask, inputType=FunctionInput_Scalar}) {name=FXMask, nodeColor=<rgba 0 0.3906 1 1>}
  NamedRerouteDeclaration(None=@n1.Normal) {name=Normal, nodeColor=<rgba 1 0.375 0 1>}
  n3 = NamedRerouteUsage()
  n4 = TextureSampleParameter2D() {parameterName=FX_HandMask, samplerType=SAMPLERTYPE_Masks, texture=&T_BlackMask}
  n5 = MaterialFunctionCall!(FXHandMask=@n4.G, ps_postion=FunctionInput() {bUsePreviewValueAsDefault=true, inputName=PS_Position}) {materialFunction=&/MF_FlowingNormals_Bites}
  n6 = MaterialFunctionCall(BaseNormal=@n5.LiquidNormal, AdditionalNormal=@n5.LiquidNormal) {materialFunction=&BlendAngleCorrectedNormals}
  n7 = LinearInterpolate(
   A=ScalarParameter() {defaultValue=0.6, group=LiquidNormal, parameterName=FlattenNormal_min}
   B=ScalarParameter() {defaultValue=0.9, group=LiquidNormal, parameterName=FlattenNormal_Max}
   Alpha=@n4.R
  ) {constA=0.6, constB=0.9}
  n8 = MaterialFunctionCall(
   BaseNormal=@n3
   AdditionalNormal=MaterialFunctionCall(
    Normal=MaterialFunctionCall(BaseNormal=@n5.WP_normal, AdditionalNormal=@n5.WP_normal) {materialFunction=&BlendAngleCorrectedNormals}.Result
    Flatness=@n7
   ) {materialFunction=&FlattenNormal}.Result
  ) {materialFunction=&BlendAngleCorrectedNormals}
  n9 = LinearInterpolate(
   A=MaterialFunctionCall(
    BaseNormal=@n3
    AdditionalNormal=MaterialFunctionCall(Normal=@n6.Result, Flatness=@n7) {materialFunction=&FlattenNormal}.Result
   ) {materialFunction=&BlendAngleCorrectedNormals}.Result
   B=@n8.Result
   Alpha=@n4.R
  )
  n10 = NamedRerouteUsage()
  NamedRerouteDeclaration(None=LinearInterpolate(A=@n3, B=QualitySwitch(Default=@n9, Low=@n8.Result, High=@n8.Result, Medium=@n8.Result, Epic=@n9), Alpha=ComponentMask(None=@n10) {g=true})) {name=OutPutNormal, nodeColor=<rgba 0 1 0.5781 1>}
  n11 = NamedRerouteUsage()
  n12 = Fresnel(ExponentIn=Constant() {r=3}, Normal=Transform(None=NamedRerouteUsage()))
  n13 = Desaturation(
   None=LinearInterpolate(
    A=TextureSampleParameter2D(UVs=@n12) {parameterName=Param, texture=/Game/Global/Textures/Default/Thinfilm/T_ThinFilm_Spectrum_COLOR}.RGB
    Alpha=Constant() {r=0.85}
   )
   Fraction=Constant() {r=0.7}
  )
  n14 = Constant3Vector()
  n15 = ScalarParameter() {defaultValue=3, group=LiquidNormal, parameterName=Brightness}
  n16 = Multiply(
   A=Saturate(None=LinearInterpolate(A=@n14.R, B=@n11, Alpha=@n15))
   B=Power(Base=FunctionInput() {inputName=AO, previewValue=<1 1 1 1>}, Exp=Constant() {r=1.3})
  )
  NamedRerouteDeclaration(None=LinearInterpolate(A=@n11, B=Multiply(A=@n13, B=@n16) {constB=0.05}, Alpha=NamedRerouteUsage())) {name=Output_basecolor, nodeColor=<rgba 0.4766 0 1 1>}
 ]}
